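# Physics asset for the Dark Viking pickaxe model, a cosmetic melee weapon.
# A single body on the Melee_TwoHanded_Handle bone uses one large capsule, offset along the bone and rotated, as its collision shape.
# It has no constraints, so the asset describes the weapon as one piece with no jointed or separately simulated parts.
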PhysicsAsset /BRCosmetics/Weapons/FORT_Melee/Pickaxe_Dark_Viking/Models/Pickaxe_Dark_Viking_PhysicsAsset
bodies {Melee_TwoHanded_Handle [{shapeType=Capsule, length=73.3739, radius=48.1447, center=<0.6696 1.4983 29.3393>, rotation=<rot -0.6655 -90 0.0791>}]}
constraints []
